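; BP_PlayerPawn_Willow is the Willow NPC pawn Blueprint, which removes itself during daylight.
; BeginPlay calls the parent, then starts a looping five-second timer that runs ValidateTimeOfDay.
; That event reads the current day phase from FortKismetLibrary and force-kills the pawn during Morning or Day.
; Tick and overlap events are disabled stubs.
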(event EventBeginPlay
  (|CallParentFunction)
  (Utilities|Time|SetTimerbyEvent (delegate Custom|ValidateTimeOfDay) 5.0 true))

(event Collision|EventActorBeginOverlap
  (params (OtherActor :type Actor%20Object%20Reference))
  (|CallParentFunction OtherActor) ; This node is disabled and will not be called.\nDrag off pins to build functionality.
)

(event Custom|ValidateTimeOfDay
  (if (or (Utilities|Enum|Equal(Enum) (TimeOfDay|GetCurrentDayPhase :!class /Script/FortniteGame.FortKismetLibrary) "Morning") (Utilities|Enum|Equal(Enum) (TimeOfDay|GetCurrentDayPhase :!class /Script/FortniteGame.FortKismetLibrary) "Day"))
    (FortPawn|ForceKill :self self)))

(event EventTick
  (params (DeltaSeconds :type Float%20%28single-precision%29))
  (|CallParentFunction DeltaSeconds) ; This node is disabled and will not be called.\nDrag off pins to build functionality.
)

(fn ConstructionScript
  (|CallParentFunction))
